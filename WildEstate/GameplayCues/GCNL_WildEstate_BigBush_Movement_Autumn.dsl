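; Looping gameplay cue for a player moving in an autumn big bush in WildEstate, adding leaf VFX and foliage sounds.
; On application, it tints the Niagara component's LeafBaseColor to a fixed brown.
; OnBurst registers the foliage sound library on the pawn and plays the bush enter sound; OnCeaseRelevant plays the exit sound and removes the library.
; OnBecomeRelevant stores the target pawn, while the tick and overlap events are empty.
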(event Collision|EventActorBeginOverlap
  (params (OtherActor :type Actor%20Object%20Reference)))

(event EventTick
  (params (DeltaSeconds :type Float%20%28single-precision%29)))

(event EventOnApplication(EitherVFX-preferthisone)
  (params (MyTarget :type Actor%20Object%20Reference) (Parameters :type Gameplay%20Cue%20Parameters%20Structure%20%28by%20ref%29) (ParticleComponents :type Array%20of%20FXSystem%20Component%20Object%20References) (AudioComponents :type Array%20of%20Audio%20Component%20Object%20References) (BurstCameraShakeInstance :type Camera%20Shake%20Base%20Object%20Reference) (BurstDecalInstance :type Decal%20Actor%20Object%20Reference))
  (bind _asniagara_particle_system_component (Utilities|Casting|CastToNiagaraComponent (Utilities|Array|Get(acopy) ParticleComponents)))
  (Niagara|SetNiagaraVariable(LinearColor) _asniagara_particle_system_component "LeafBaseColor" (0.104167 0.054445 0.024007 1.0)))

(fn OnBurst
  (params (MyTarget :type Actor%20Object%20Reference) (Parameters :type Gameplay%20Cue%20Parameters%20Structure%20%28by%20ref%29))
  (bind _asfort_player_pawn_athena (Utilities|Casting|CastToFortPlayerPawnAthena MyTarget))
  (SoundLibrary|AddLibrary (WorldSubsystems|GetSoundLibrarySubsystem _asfort_player_pawn_athena) _asfort_player_pawn_athena "/Game/Sounds/Fort_Foley_Sounds/Foliage/Settings/Foliage_Lib.Foliage_Lib" :!class /Script/SoundLibrary.SoundLibrarySubsystem)
  (SoundLibrary|PlaySoundSimple (WorldSubsystems|GetSoundLibrarySubsystem _asfort_player_pawn_athena) _asfort_player_pawn_athena (Utilities|Struct|MakeSimpleSoundSettings "(TagName=\"SoundLibrary.Player.State.Bush.Enter\")" false false true false 2.3) :!class /Script/SoundLibrary.SoundLibrarySubsystem)
  (return false))

(fn OnCeaseRelevant
  (params (MyTarget :type Actor%20Object%20Reference) (Parameters :type Gameplay%20Cue%20Parameters%20Structure%20%28by%20ref%29))
  (Utilities|IsValid MyTarget
    (:Is%20Valid
      (bind _asfort_player_pawn_athena (Utilities|Casting|CastToFortPlayerPawnAthena MyTarget))
      (SoundLibrary|PlaySoundSimple (WorldSubsystems|GetSoundLibrarySubsystem _asfort_player_pawn_athena) _asfort_player_pawn_athena (Utilities|Struct|MakeSimpleSoundSettings "(TagName=\"SoundLibrary.Player.State.Bush.Exit\")" false false true false 2.3) :!class /Script/SoundLibrary.SoundLibrarySubsystem)
      (SoundLibrary|RemoveLibrary (WorldSubsystems|GetSoundLibrarySubsystem _asfort_player_pawn_athena) _asfort_player_pawn_athena "/Game/Sounds/Fort_Foley_Sounds/Foliage/Settings/Foliage_Lib.Foliage_Lib" :!class /Script/SoundLibrary.SoundLibrarySubsystem)
      (return false))))

(fn ConstructionScript)

(fn OnBecomeRelevant
  (params (MyTarget :type Actor%20Object%20Reference) (Parameters :type Gameplay%20Cue%20Parameters%20Structure%20%28by%20ref%29))
  (bind _asfort_player_pawn (Utilities|Casting|CastToFortPlayerPawn MyTarget))
  (Variables|Default|SetTargetPawn _asfort_player_pawn)
  (return false))
