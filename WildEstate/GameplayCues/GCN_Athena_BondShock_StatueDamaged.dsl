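; Burst gameplay cue that plays impact audio when a Bond Shock statue is damaged.
; After the parent burst runs, a valid instigator with the harvest melee equipped tag gets a pickaxe impact sound at the hit location.
; Otherwise the instigator is cast to a player controller, and its Athena pawn's locally-viewed state selects a 1P or 3P weapon impact sound.
; A locally viewed pawn also activates the Bond Shock control bus mix.
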(fn OnBurst(EitherVFX-preferthisversion)
  (params (MyTarget :type Actor%20Object%20Reference) (Parameters :type Gameplay%20Cue%20Parameters%20Structure%20%28by%20ref%29) (ParticleComponents :type Array%20of%20FXSystem%20Component%20Object%20References) (AudioComponents :type Array%20of%20Audio%20Component%20Object%20References) (BurstCameraShakeInstance :type Camera%20Shake%20Base%20Object%20Reference) (BurstDecalInstance :type Decal%20Actor%20Object%20Reference))
  (bind (_normalizedmagnitude _rawmagnitude _effectcontext _matchedtagname _originaltag _aggregatedsourcetags _aggregatedtargettags _location _normal _instigator) (Ability|GameplayCue|BreakGameplayCueParameters Parameters))
  (|CallParentFunction MyTarget Parameters ParticleComponents AudioComponents BurstCameraShakeInstance BurstDecalInstance)
  (Utilities|IsValid _instigator
    (:Is%20Valid
      (bind _returnvalue (GameplayTags|HasMatchingGameplayTag(Message) _instigator "(TagName=\"Granted.ItemEquipped.Melee.Harvest\")"))
      (if _returnvalue
        (Audio|SpawnSoundatLocation _instigator (Variables|Default|GetPickaxeImpactSound) _location)
        (else
          (bind _asfort_player_controller (Utilities|Casting|CastToFortPlayerController _instigator)
            (:then
              (bind _asfort_player_pawn_athena (Utilities|Casting|CastToFortPlayerPawnAthena (Game|Player|GetPlayerPawn _asfort_player_controller))
                (:then
                  (if (Pawn|IsLocallyViewed _asfort_player_pawn_athena)
                    (Audio|Modulation|ActivateControlBusMix MyTarget (Variables|Default|GetCBM_BondShock))
                    (else
                      _))
                  (Audio|SpawnSoundatLocation _instigator (select (Pawn|IsLocallyViewed _asfort_player_pawn_athena) (Variables|Default|GetWeaponImpactSound_1P) (Variables|Default|GetWeaponImpactSound_3P)) _location)
                  (return))
                (:CastFailed
                  _)))
            (:CastFailed
              _)))))))
